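# One-shot stereo MetaSound source for the Stalker Knife's air attack sound, with a flight layer and an outro.
# On play, the StalkerKnife_Fly wave starts and an ADSR envelope on its mixer gain fades it in.
# PlayOutro releases that envelope and plays a random pick from a wave array at fixed gain.
# The source reports finished only when that outro wave ends, so it relies on PlayOutro being triggered.
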MetaSoundSource /FNMStalkerKnife/Assets/Audio/MSS_StalkerKnife_AirAttack_Lp
inputs [
 PlayOutro {type=Trigger}
 UE.Source.OnPlay {type=Trigger}
]
interfaces [UE.OutputFormat.Stereo:1.0, UE.Source.OneShot:1.0, UE.Source:1.1]
outputs [
 "UE.OutputFormat.Stereo.Audio:0" {type=Audio}
 "UE.OutputFormat.Stereo.Audio:1" {type=Audio}
 UE.Source.OneShot.OnFinished {type=Trigger}
]
pages [Default {
  input_defaults {PlayOutro=null, UE.Source.OnPlay=null}
  nodes [
   TriggerDelay_0 = UE.Trigger Delay(In=^UE.Source.OnPlay) {also {defaults {Delay Time=0}}}
   ADSREnvelopeADSREnvelope_0 = ADSR Envelope.ADSR Envelope:Float(Trigger Release=^PlayOutro, Trigger Attack=@TriggerDelay_0.Out) {also {defaults {Attack Curve=2, Attack Time=2, Release Curve=-1, Release Time=0.4, Sustain Level=1}}}
   ArrayRandomGet_0 = Array.Random Get:WaveAsset:Array(Next=^PlayOutro, In Array=VariableAccessor.WaveAsset:Array(Variable=InitVariable.WaveAsset:Array().Variable).Value)
   WavePlayer_0 = UE.Wave Player:Stereo(Play=@ArrayRandomGet_0.On Next, Wave Asset=@ArrayRandomGet_0.Value) {also {version [1,1]}}
   WavePlayer_1 = UE.Wave Player:Stereo(Play=@TriggerDelay_0.Out) {also {defaults {Loop="<MetaSoundBuilderResult.SUCCEEDED: 0>", Wave Asset=/FNMStalkerKnife/Assets/Audio/Waves/StalkerKnife_Fly}, version [1,1]}}
   AudioMixerAudioMixerStereo2_0 = "AudioMixer.Audio Mixer (Stereo, 2)"(
    Gain 1=@ADSREnvelopeADSREnvelope_0.Out Envelope
    In 0 L=@WavePlayer_0.Out Left
    In 0 R=@WavePlayer_0.Out Right
    In 1 L=@WavePlayer_1.Out Left
    In 1 R=@WavePlayer_1.Out Right
   ) {also {defaults {Gain 0=0.6}}}
   InitVariable.WaveAsset:Array()
   InitVariable.WaveAsset:Array()
  ]
  links [
   {dst="@UE.OutputFormat.Stereo.Audio:0", src=AudioMixerAudioMixerStereo2_0.Out L}
   {dst="@UE.OutputFormat.Stereo.Audio:1", src=AudioMixerAudioMixerStereo2_0.Out R}
   {dst="@UE.Source.OneShot.OnFinished", src=WavePlayer_0.On Finished}
  ]
 }]
properties {
 assetImportData=&self:AssetImportData
 attenuationSettings=&/ATT_Fort_StalkerKnife_Attack
 maxDistance=4100
 numChannels=2
 outputFormat=Stereo
 sampleRate=48000
 soundClassObject=&Fort_Gadgets
}
