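# Stereo one-shot MetaSound source for the Stalker Knife attack sound, using melee attenuation and the gadgets sound class.
# On play, a short trigger delay fires a random pick from a wave array variable.
# A random integer, with Min set above Max, is converted to float as the wave player's pitch shift.
# The player feeds both outputs and signals finish; the Fort character inputs are unused.
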MetaSoundSource /FNMStalkerKnife/Assets/Audio/MSS_StalkerKnife_Attack_C
inputs [
 Character.Affiliation {type=Enum:FortAudioAffiliation}
 Character.Health {type=Float}
 Character.UniqueID {type=Int32}
 UE.Source.OnPlay {type=Trigger}
]
interfaces [Fort.Character:1.0, UE.OutputFormat.Stereo:1.0, UE.Source.OneShot:1.0, UE.Source:1.1]
outputs [
 "UE.OutputFormat.Stereo.Audio:0" {type=Audio}
 "UE.OutputFormat.Stereo.Audio:1" {type=Audio}
 UE.Source.OneShot.OnFinished {type=Trigger}
]
pages [Default {
  input_defaults {Character.Affiliation=0, Character.Health=0, Character.UniqueID=0, UE.Source.OnPlay=null}
  nodes [
   TriggerDelay_0 = UE.Trigger Delay(In=^UE.Source.OnPlay) {also {defaults {Delay Time=0.38}}}
   ArrayRandomGet_0 = Array.Random Get:WaveAsset:Array(Next=@TriggerDelay_0.Out, In Array=VariableAccessor.WaveAsset:Array(Variable=InitVariable.WaveAsset:Array().Variable).Value) {also {defaults {No Repeats=-1, Weights []}}}
   n1 = UE.RandomInt32(Next=@ArrayRandomGet_0.On Next) {also {defaults {Max=1, Min=2}, version [1,1]}}
   WavePlayer_0 = UE.Wave Player:Stereo(Wave Asset=@ArrayRandomGet_0.Value, Pitch Shift=Convert.Float:Int32(Int32=@n1.Value).Float, Play=@n1.On Next) {also {version [1,1]}}
  ]
  links [
   {dst="@UE.Source.OneShot.OnFinished", src=WavePlayer_0.On Finished}
   {dst="@UE.OutputFormat.Stereo.Audio:0", src=WavePlayer_0.Out Left}
   {dst="@UE.OutputFormat.Stereo.Audio:1", src=WavePlayer_0.Out Right}
  ]
 }]
properties {
 assetImportData=&self:AssetImportData
 attenuationSettings=/Game/Sounds/Attenuation/Fort_Wasabi_Melee_Att
 maxDistance=3800
 numChannels=2
 outputFormat=Stereo
 sampleRate=48000
 soundClassObject=&Fort_Gadgets
 volume=0.8
}
