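# Stereo one-shot MetaSound for the Stalker Knife player-spotted cue, using enemy-spotted attenuation.
# On play, it compares the character's affiliation to choose between two random wave pickers.
# A match plays immediately, while a mismatch waits for a short delay before playing from its own array.
# Both players feed a mixer with separate gains, and either one finishing ends the sound.
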MetaSoundSource /FNMStalkerKnife/Assets/Audio/MSS_StalkerKnife_Player_Spotted
inputs [
 Character.Affiliation {type=Enum:FortAudioAffiliation}
 Character.Health {type=Float}
 Character.UniqueID {type=Int32}
 UE.Source.OnPlay {type=Trigger}
]
interfaces [Fort.Character:1.0, UE.OutputFormat.Stereo:1.0, UE.Source.OneShot:1.0, UE.Source:1.1]
outputs [
 "UE.OutputFormat.Stereo.Audio:0" {type=Audio}
 "UE.OutputFormat.Stereo.Audio:1" {type=Audio}
 UE.Source.OneShot.OnFinished {type=Trigger}
]
pages [Default {
  input_defaults {Character.Affiliation=0, Character.Health=0, Character.UniqueID=0, UE.Source.OnPlay=null}
  nodes [
   n1 = &TriggerCompareAffiliation(A=^Character.Affiliation, Compare=^UE.Source.OnPlay)
   ArrayRandomGet_1 = Array.Random Get:WaveAsset:Array(Next=@n1.True, In Array=VariableAccessor.WaveAsset:Array(Variable=InitVariable.WaveAsset:Array().Variable).Value)
   WavePlayer_0 = UE.Wave Player:Stereo(Play=@ArrayRandomGet_1.On Next, Wave Asset=@ArrayRandomGet_1.Value) {also {version [1,1]}}
   TriggerDelay_0 = UE.Trigger Delay(In=@n1.False) {also {defaults {Delay Time=0.5}}}
   ArrayRandomGet_0 = Array.Random Get:WaveAsset:Array(Next=@TriggerDelay_0.Out, In Array=VariableAccessor.WaveAsset:Array(Variable=InitVariable.WaveAsset:Array().Variable).Value) {also {defaults {Weights []}}}
   WavePlayer_1 = UE.Wave Player:Stereo(Play=@ArrayRandomGet_0.On Next, Wave Asset=@ArrayRandomGet_0.Value) {also {version [1,1]}}
   AudioMixerAudioMixerStereo4_0 = "AudioMixer.Audio Mixer (Stereo, 4)"(In 1 L=@WavePlayer_0.Out Left, In 1 R=@WavePlayer_0.Out Right, In 2 L=@WavePlayer_1.Out Left, In 2 R=@WavePlayer_1.Out Right) {also {defaults {Gain 0=0.6, Gain 1=1.7, Gain 2=1.7, Gain 3=0.25}}}
   InitVariable.WaveAsset:Array()
   InitVariable.WaveAsset:Array()
   TriggerAnyTriggerAny2_0 = "TriggerAny.Trigger Any (2)"(In 0=@WavePlayer_0.On Finished, In 1=@WavePlayer_1.On Finished)
  ]
  links [
   {dst="@UE.OutputFormat.Stereo.Audio:0", src=AudioMixerAudioMixerStereo4_0.Out L}
   {dst="@UE.OutputFormat.Stereo.Audio:1", src=AudioMixerAudioMixerStereo4_0.Out R}
   {dst="@UE.Source.OneShot.OnFinished", src=TriggerAnyTriggerAny2_0.Out}
  ]
 }]
properties {
 assetImportData=&self:AssetImportData
 attenuationSettings=&/ATT_Fort_StalkerKnife_EnemySpoted
 bOverrideConcurrency=true
 maxDistance=17000
 numChannels=2
 outputFormat=Stereo
 sampleRate=48000
 soundClassObject=&Fort_Gadgets
 volume=0.5
 concurrencyOverrides {maxCount=5}
}
