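# One-shot stereo MetaSound for the Stalker Knife player-spotted HUD cue, with attenuation and a concurrency override.
# On play, it compares the character's affiliation to a fixed value to choose between two marking sounds.
# A match plays the "targeted by opponent" wave and a mismatch the "opponent caret placement" wave, each after a short delay.
# Both players feed a mixer with separate gains, and either one finishing ends the source.
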MetaSoundSource /FNMStalkerKnife/Assets/Audio/MSS_StalkerKnife_Player_Spotted_HUD
inputs [
 Character.Affiliation {type=Enum:FortAudioAffiliation}
 Character.Health {type=Float}
 Character.UniqueID {type=Int32}
 UE.Source.OnPlay {type=Trigger}
]
interfaces [Fort.Character:1.0, UE.OutputFormat.Stereo:1.0, UE.Source.OneShot:1.0, UE.Source:1.1]
outputs [
 "UE.OutputFormat.Stereo.Audio:0" {type=Audio}
 "UE.OutputFormat.Stereo.Audio:1" {type=Audio}
 UE.Source.OneShot.OnFinished {type=Trigger}
]
pages [Default {
  input_defaults {Character.Affiliation=0, Character.Health=0, Character.UniqueID=0, UE.Source.OnPlay=null}
  nodes [
   n1 = &TriggerCompareAffiliation(A=^Character.Affiliation, Compare=^UE.Source.OnPlay) {also {defaults {B=1}}}
   TriggerDelay_0 = UE.Trigger Delay(In=@n1.False) {also {defaults {Delay Time=1.5}}}
   WavePlayer_0 = UE.Wave Player:Stereo(Play=@TriggerDelay_0.Out) {also {defaults {Wave Asset=/Game/Sounds/Fort_UI_Sounds/Gameplay/Marking/HUD_Marked_Opponent_Caret_Placement_01}, version [1,1]}}
   TriggerDelay_1 = UE.Trigger Delay(In=@n1.True) {also {defaults {Delay Time=1.4}}}
   WavePlayer_1 = UE.Wave Player:Stereo(Play=@TriggerDelay_1.Out) {also {defaults {Wave Asset=/Game/Sounds/Fort_UI_Sounds/Gameplay/Marking/HUD_Marked_Targeted_By_Opponent_01}, version [1,1]}}
   AudioMixerAudioMixerStereo2_0 = "AudioMixer.Audio Mixer (Stereo, 2)"(In 1 L=@WavePlayer_0.Out Left, In 1 R=@WavePlayer_0.Out Right, In 0 L=@WavePlayer_1.Out Left, In 0 R=@WavePlayer_1.Out Right) {also {defaults {Gain 0=0.6, Gain 1=0.25}}}
   InitVariable.WaveAsset:Array()
   InitVariable.WaveAsset:Array()
   InitVariable.WaveAsset:Array()
   InitVariable.WaveAsset:Array()
   TriggerAnyTriggerAny2_0 = "TriggerAny.Trigger Any (2)"(In 0=UE.Reroute(Value=@WavePlayer_1.On Finished).Value, In 1=UE.Reroute(Value=UE.Reroute(Value=@WavePlayer_0.On Finished).Value).Value)
  ]
  links [
   {dst="@UE.OutputFormat.Stereo.Audio:0", src=AudioMixerAudioMixerStereo2_0.Out L}
   {dst="@UE.OutputFormat.Stereo.Audio:1", src=AudioMixerAudioMixerStereo2_0.Out R}
   {dst="@UE.Source.OneShot.OnFinished", src=TriggerAnyTriggerAny2_0.Out}
  ]
 }]
properties {
 assetImportData=&self:AssetImportData
 attenuationSettings=&/ATT_Fort_StalkerKnife_EnemySpoted
 bOverrideConcurrency=true
 maxDistance=17000
 numChannels=2
 outputFormat=Stereo
 sampleRate=48000
 soundClassObject=&Fort_Gadgets
 volume=1.2
 concurrencyOverrides {maxCount=8}
}
