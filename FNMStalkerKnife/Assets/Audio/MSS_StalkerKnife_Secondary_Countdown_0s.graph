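# One-shot stereo MetaSound source for the Stalker Knife secondary countdown's zero-second cue, in the melee weapon sound class.
# On play, a single stereo Wave Player plays the countdown wave straight to the left and right outputs.
# The source reports finished when the wave completes.
# The Fort character interface inputs (affiliation, health, unique ID) are declared but unused.
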MetaSoundSource /FNMStalkerKnife/Assets/Audio/MSS_StalkerKnife_Secondary_Countdown_0s
inputs [
 Character.Affiliation {type=Enum:FortAudioAffiliation}
 Character.Health {type=Float}
 Character.UniqueID {type=Int32}
 UE.Source.OnPlay {type=Trigger}
]
interfaces [Fort.Character:1.0, UE.OutputFormat.Stereo:1.0, UE.Source.OneShot:1.0, UE.Source:1.1]
outputs [
 "UE.OutputFormat.Stereo.Audio:0" {type=Audio}
 "UE.OutputFormat.Stereo.Audio:1" {type=Audio}
 UE.Source.OneShot.OnFinished {type=Trigger}
]
pages [Default {
  input_defaults {Character.Affiliation=0, Character.Health=0, Character.UniqueID=0, UE.Source.OnPlay=null}
  nodes [WavePlayer_0 = UE.Wave Player:Stereo(Play=^UE.Source.OnPlay) {also {defaults {Wave Asset=/FNMStalkerKnife/Assets/Audio/Waves/StalkerKnife_Secondary_Countdown_0s}, version [1,1]}}]
  links [
   {dst="@UE.Source.OneShot.OnFinished", src=WavePlayer_0.On Finished}
   {dst="@UE.OutputFormat.Stereo.Audio:0", src=WavePlayer_0.Out Left}
   {dst="@UE.OutputFormat.Stereo.Audio:1", src=WavePlayer_0.Out Right}
  ]
 }]
properties {
 assetImportData=&self:AssetImportData
 numChannels=2
 outputFormat=Stereo
 sampleRate=48000
 soundClassObject=&Fort_WeaponMelee
}
