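# One-shot stereo MetaSound for the Stalker Knife's Stalker Mode activation, branching playback on the character's audio affiliation.
# An affiliation match plays one wave player after a short fixed delay.
# Otherwise, a second player fires after the DelayTime input plus a fixed delay, with a randomized integer pitch shift.
# Both players feed one mixer and either finishing ends the sound; a PrintInt patch outputs the affiliation on play.
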MetaSoundSource /FNMStalkerKnife/Assets/Audio/MSS_StalkerKnife_StalkerMode_Activation
inputs [
 Character.Affiliation {type=Enum:FortAudioAffiliation}
 Character.Health {type=Float}
 Character.UniqueID {type=Int32}
 DelayTime {type=Float}
 UE.Source.OnPlay {type=Trigger}
]
interfaces [Fort.Character:1.0, UE.OutputFormat.Stereo:1.0, UE.Source.OneShot:1.0, UE.Source:1.1]
outputs [
 "UE.OutputFormat.Stereo.Audio:0" {type=Audio}
 "UE.OutputFormat.Stereo.Audio:1" {type=Audio}
 UE.Source.OneShot.OnFinished {type=Trigger}
]
pages [Default {
  input_defaults {Character.Affiliation=0, Character.Health=0, Character.UniqueID=0, DelayTime=null, UE.Source.OnPlay=null}
  nodes [
   n1 = &TriggerCompareAffiliation(A=^Character.Affiliation, Compare=^UE.Source.OnPlay)
   TriggerDelay_0 = UE.Trigger Delay(In=@n1.True) {also {defaults {Delay Time=0.2}}}
   WavePlayer_0 = UE.Wave Player:Stereo(Play=@TriggerDelay_0.Out, Wave Asset=VariableAccessor.WaveAsset(Variable=InitVariable.WaveAsset().Variable).Value) {also {version [1,1]}}
   TriggerDelay_2 = UE.Trigger Delay(Delay Time=UE.ConversionFloatToTime(In=^DelayTime).Out, In=@n1.False)
   n2 = UE.RandomInt32(Next=@TriggerDelay_2.Out) {also {defaults {Max=0, Min=-1}, version [1,1]}}
   TriggerDelay_1 = UE.Trigger Delay(In=@n2.On Next) {also {defaults {Delay Time=0.35}}}
   WavePlayer_1 = UE.Wave Player:Stereo(
    Pitch Shift=Convert.Float:Int32(Int32=@n2.Value).Float
    Play=@TriggerDelay_1.Out
    Wave Asset=VariableAccessor.WaveAsset(Variable=InitVariable.WaveAsset().Variable).Value
   ) {also {version [1,1]}}
   AudioMixerAudioMixerStereo2_0 = "AudioMixer.Audio Mixer (Stereo, 2)"(In 0 L=@WavePlayer_0.Out Left, In 0 R=@WavePlayer_0.Out Right, In 1 L=@WavePlayer_1.Out Left, In 1 R=@WavePlayer_1.Out Right)
   /FortMetasound/Patches/MSP_PrintInt(Trigger=^UE.Source.OnPlay, Input=Convert.Int32:Enum:FortAudioAffiliation("Enum:FortAudioAffiliation"=^Character.Affiliation).Int32)
   TriggerAnyTriggerAny2_0 = "TriggerAny.Trigger Any (2)"(In 0=@WavePlayer_0.On Finished, In 1=@WavePlayer_1.On Finished)
  ]
  links [
   {dst="@UE.OutputFormat.Stereo.Audio:0", src=AudioMixerAudioMixerStereo2_0.Out L}
   {dst="@UE.OutputFormat.Stereo.Audio:1", src=AudioMixerAudioMixerStereo2_0.Out R}
   {dst="@UE.Source.OneShot.OnFinished", src=TriggerAnyTriggerAny2_0.Out}
  ]
 }]
properties {
 assetImportData=&self:AssetImportData
 attenuationSettings=&/ATT_Fort_StalkerKnife_StalkerMode_Activation
 maxDistance=20000
 numChannels=2
 outputFormat=Stereo
 sampleRate=48000
 soundClassObject=&Fort_Gadgets
 volume=1.65
}
