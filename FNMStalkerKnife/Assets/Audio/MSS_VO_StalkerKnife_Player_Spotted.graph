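# One-shot stereo MetaSound for the Stalker Knife "player spotted" NPC voice line.
# On play it branches on a character affiliation comparison, and each branch waits on a fixed delay before playing a random wave from an array variable.
# A mixer boosts and combines both players, and either finishing ends the source.
# Concurrency allows one instance without interruption, and bus sends feed sidechain buses for other sound categories.
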MetaSoundSource /FNMStalkerKnife/Assets/Audio/MSS_VO_StalkerKnife_Player_Spotted
inputs [
 Character.Affiliation {type=Enum:FortAudioAffiliation}
 Character.Health {type=Float}
 Character.UniqueID {type=Int32}
 UE.Source.OnPlay {type=Trigger}
]
interfaces [Fort.Character:1.0, UE.OutputFormat.Stereo:1.0, UE.Source.OneShot:1.0, UE.Source:1.1]
outputs [
 "UE.OutputFormat.Stereo.Audio:0" {type=Audio}
 "UE.OutputFormat.Stereo.Audio:1" {type=Audio}
 UE.Source.OneShot.OnFinished {type=Trigger}
]
pages [Default {
  input_defaults {Character.Affiliation=0, Character.Health=0, Character.UniqueID=0, UE.Source.OnPlay=null}
  nodes [
   n1 = &TriggerCompareAffiliation(A=^Character.Affiliation, Compare=^UE.Source.OnPlay)
   TriggerDelay_1 = UE.Trigger Delay(In=@n1.True) {also {defaults {Delay Time=2.2}}}
   ArrayRandomGet_1 = Array.Random Get:WaveAsset:Array(Next=@TriggerDelay_1.Out, In Array=VariableAccessor.WaveAsset:Array(Variable=InitVariable.WaveAsset:Array().Variable).Value) {also {defaults {Weights []}}}
   WavePlayer_0 = UE.Wave Player:Stereo(Play=@ArrayRandomGet_1.On Next, Wave Asset=@ArrayRandomGet_1.Value) {also {version [1,1]}}
   TriggerDelay_0 = UE.Trigger Delay(In=@n1.False) {also {defaults {Delay Time=2.2}}}
   ArrayRandomGet_0 = Array.Random Get:WaveAsset:Array(Next=@TriggerDelay_0.Out, In Array=VariableAccessor.WaveAsset:Array(Variable=InitVariable.WaveAsset:Array().Variable).Value) {also {defaults {Weights []}}}
   WavePlayer_1 = UE.Wave Player:Stereo(Play=@ArrayRandomGet_0.On Next, Wave Asset=@ArrayRandomGet_0.Value) {also {version [1,1]}}
   AudioMixerAudioMixerStereo2_0 = "AudioMixer.Audio Mixer (Stereo, 2)"(In 0 L=@WavePlayer_0.Out Left, In 0 R=@WavePlayer_0.Out Right, In 1 L=@WavePlayer_1.Out Left, In 1 R=@WavePlayer_1.Out Right) {also {defaults {Gain 0=1.8, Gain 1=1.8}}}
   InitVariable.WaveAsset:Array()
   InitVariable.WaveAsset:Array()
   TriggerAnyTriggerAny2_0 = "TriggerAny.Trigger Any (2)"(In 1=UE.Reroute(Value=@WavePlayer_1.On Finished).Value, In 0=UE.Reroute(Value=@WavePlayer_0.On Finished).Value)
  ]
  links [
   {dst="@UE.OutputFormat.Stereo.Audio:0", src=AudioMixerAudioMixerStereo2_0.Out L}
   {dst="@UE.OutputFormat.Stereo.Audio:1", src=AudioMixerAudioMixerStereo2_0.Out R}
   {dst="@UE.Source.OneShot.OnFinished", src=TriggerAnyTriggerAny2_0.Out}
  ]
 }]
properties {
 assetImportData=&self:AssetImportData
 bOverrideConcurrency=true
 numChannels=2
 outputFormat=Stereo
 sampleRate=48000
 soundClassObject=&Fort_VO_NPC
 busSends [
  {audioBus=&AB_Sidechain_FortWeapon_Vol, sendLevel=1}
  {audioBus=/Game/Sounds/Modulation/EnvelopeFollowers/Volume/AB_Sidechain_Footsteps_Vol, sendLevel=60}
  {audioBus=/Game/Sounds/Modulation/EnvelopeFollowers/Volume/AB_SideChain_FortAmbient_Outdoors_Vol, sendLevel=1}
  {audioBus=&AB_Sidechain_FortAmbient_Vol, sendLevel=1}
  {audioBus=&AB_Sidechain_FortExplosion_Vol, sendLevel=1}
  {audioBus=&AB_Sidechain_FortMusic_Vol, sendLevel=1}
  {audioBus=&AB_Sidechain_FortVehicles_Vol, sendLevel=1}
  {audioBus=/Game/Sounds/Modulation/EnvelopeFollowers/Volume/AB_Sidechain_FortWeapon1P_Vol, sendLevel=1}
 ]
 concurrencyOverrides {maxCount=1, resolutionRule=PreventNew, retriggerTime=5}
}
