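# Sound cue named as the Stalker Knife's dial tone, though the wave it plays is the knife's activation sound.
# A single Wave Player is the root, feeding the wave straight out with no randomization, looping or modulation nodes.
# The cue boosts volume well above normal and sets a very large max distance.
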SoundCue /FNMStalkerKnife/Assets/Audio/SC_StalkerKnife_DialTone
root=WavePlayer_0
nodes [{id=WavePlayer_0, properties {soundWaveAssetPtr=/FNMStalkerKnife/Assets/Audio/Waves/StalkerKnife_Activation}}]
properties {maxDistance=2097152, volumeMultiplier=4}
