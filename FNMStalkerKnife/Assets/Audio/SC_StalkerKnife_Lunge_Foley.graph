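# Sound Cue for the foley of the Stalker Knife's lunge.
# A Random node picks among six equally weighted foley wave variations and outputs directly, with a boosted cue volume.
# Inline attenuation overrides add distance falloff and occlusion on a custom trace channel, superseding the referenced StationProp attenuation asset.
# A concurrency override caps how many instances play at once.
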SoundCue /FNMStalkerKnife/Assets/Audio/SC_StalkerKnife_Lunge_Foley
root=Random_0
nodes [
 {id=Random_0, children [WavePlayer_1,WavePlayer_2,WavePlayer_3,WavePlayer_4,WavePlayer_5,WavePlayer_6], properties {weights [1,1,1,1,1,1]}}
 {id=WavePlayer_1, properties {soundWaveAssetPtr=/FNMStalkerKnife/Assets/Audio/Waves/StalkerKnife_Lunge_Foley_01}}
 {id=WavePlayer_2, properties {soundWaveAssetPtr=/FNMStalkerKnife/Assets/Audio/Waves/StalkerKnife_Lunge_Foley_02}}
 {id=WavePlayer_3, properties {soundWaveAssetPtr=/FNMStalkerKnife/Assets/Audio/Waves/StalkerKnife_Lunge_Foley_03}}
 {id=WavePlayer_4, properties {soundWaveAssetPtr=/FNMStalkerKnife/Assets/Audio/Waves/StalkerKnife_Lunge_Foley_04}}
 {id=WavePlayer_5, properties {soundWaveAssetPtr=/FNMStalkerKnife/Assets/Audio/Waves/StalkerKnife_Lunge_Foley_05}}
 {id=WavePlayer_6, properties {soundWaveAssetPtr=/FNMStalkerKnife/Assets/Audio/Waves/StalkerKnife_Lunge_Foley_06}}
]
properties {
 attenuationSettings=/Game/Sounds/Foley_Loot/Containers/StationProp/StationProp_Default_Att
 bOverrideAttenuation=true
 bOverrideConcurrency=true
 maxDistance=1400
 volumeMultiplier=4
 attenuationOverrides {bEnableOcclusion=true, falloffDistance=1000, occlusionTraceChannel=ECC_GameTraceChannel11}
 concurrencyOverrides {maxCount=3}
}
