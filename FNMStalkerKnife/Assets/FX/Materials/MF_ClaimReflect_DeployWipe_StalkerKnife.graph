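# Material function for the StalkerKnife claim-reflect deploy wipe, outputting a WipeMask and a glowing WipeLine from a directional gradient input.
# The gradient is offset and perturbed through reroutes, then shaped by SphereMask and SmoothSteps into edge and mask.
# Timing curves derive from DeployTimestamp (scalar plus dynamic parameter) and animation length, with DebugDeployWipe blending in a looping preview.
# Local-XZ UVs, computed per vertex and noise-distorted, sample plasma and Voronoi masks.
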MaterialFunction /FNMStalkerKnife/Assets/FX/Materials/MF_ClaimReflect_DeployWipe_StalkerKnife
graph {nodes [
  Constant() {r=2}
  Constant() {r=1}
  n1 = Subtract(A=NamedRerouteUsage()) {constB=0.5}
  n2 = FunctionInput(Preview=ComponentMask(None=TextureCoordinate()) {r=true}) {bUsePreviewValueAsDefault=true, inputName=InDerectionalGradient, inputType=FunctionInput_Scalar}
  n3 = Add(A=Add(A=Multiply(A=@n1) {constB=0.1}, B=Subtract(A=@n2)), B=LinearInterpolate(Alpha=NamedRerouteUsage()) {constA=-0.2, constB=2})
  n4 = Multiply(
   A=SphereMask(A=@n3, B=Constant()) {attenuationRadius=0.02, hardnessPercent=0}
   B=ScalarParameter() {defaultValue=10, parameterName=DeployWipeLineInt}
  )
  n5 = OneMinus(None=@n3)
  n6 = SmoothStep(Value=Add(A=@n3) {constB=0.05}) {constMax=0.15}
  n7 = Multiply(A=Multiply(A=Power(Base=@n5), B=@n6)) {constB=0.5}
  n8 = NamedRerouteDeclaration(None=SmoothStep(Value=@n6) {constMax=0.2}) {name=WipeMask, nodeColor=<rgba 1 0 0 1>}
  n9 = NamedRerouteDeclaration(None=Multiply(A=Add(A=@n4, B=@n7), B=@n8)) {name=WipeLine, nodeColor=<rgba 0 1 1 1>}
  FunctionOutput(None=@n9) {outputName=WipeLine}
  FunctionOutput(None=@n8) {outputName=WipeMask}
  LinearInterpolate(A=@n9, B=@n8) {constAlpha=0.25}
  n10 = Time()
  n11 = Add(
   A=ScalarParameter() {parameterName=DeployTimestamp}
   B=DynamicParameter() {defaultValue=<rgba 0 1 1 1>, paramNames [DeployTimestamp, Param2, Param3, Param4]}.DeployTimestamp
  )
  n12 = Subtract(A=@n10, B=@n11)
  n13 = ScalarParameter() {defaultValue=1, parameterName=DeployAnimationLength}
  n14 = Multiply(A=Time()) {constB=0.5}
  n15 = LinearInterpolate(
   A=Clamp(None=Divide(A=@n12, B=@n13) {desc=InLength})
   B=NamedRerouteDeclaration(None=Frac(None=@n14)) {name=WipeTimeCurve, nodeColor=<rgba 0.3125 0 1 1>}
   Alpha=ScalarParameter() {parameterName=DebugDeployWipe}
  )
  NamedRerouteDeclaration(None=@n15) {name=DeployTimestampIn, nodeColor=<rgba 0 0.6953 1 1>}
  n16 = Subtract(A=@n10, B=Add(A=@n11, B=Add(A=@n13, B=@n13) {constB=3, desc=eXTRAdURATION}))
  n17 = OneMinus(None=Clamp(None=Divide(A=@n16) {constB=3, desc=OutLength}))
  NamedRerouteDeclaration(None=Multiply(A=@n15, B=@n17)) {name=DeployTimestampInOut, nodeColor=<rgba 1 0.9844 0 1>}
  NamedRerouteDeclaration(None=@n17) {name=DeployTimestampOut, nodeColor=<rgba 0 1 0.3438 1>}
  n18 = ComponentMask(None=LocalPosition().XYZ) {b=true, r=true}
  n19 = VertexInterpolator(VS=Multiply(A=Multiply(A=@n18) {constB=0.01}) {constB=2})
  n20 = TextureSample(UVs=@n19.PS) {texture=&AtmosphericCloudNoise01}
  n21 = Add(A=@n19.PS, B=Multiply(A=Subtract(A=@n20.G) {constB=0.5}) {constB=0.1})
  n22 = TextureSample(UVs=@n21) {samplerType=SAMPLERTYPE_Masks, texture=/FNMStalkerKnife/Assets/FX/Textures/T_Pickaxe_ClaimReflect_Mask_StalkerKnife}
  NamedRerouteDeclaration(None=@n22.R) {name=MidFreqPlasma, nodeColor=<rgba 0 0.3672 1 1>}
  NamedRerouteDeclaration(None=@n22.G) {name=MidFreqVoronoi, nodeColor=<rgba 0.6875 1 0 1>}
  Panner(Coordinate=@n21) {speedX=0.1}
 ]}
