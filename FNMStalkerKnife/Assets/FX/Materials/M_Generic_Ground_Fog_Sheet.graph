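# Unlit translucent material for Niagara sprite and mesh particles that draws drifting sheets of ground fog.
# Two cloud textures pan with time offset by actor position and U shifted per particle, then blend and square into noise.
# Noise shapes a soft spherical mask with view-angle fade and also varies the depth-fade distance.
# Particle color feeds an emissive adjustment function, while particle alpha scales opacity, which fades out far from the camera.
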Material /FNMStalkerKnife/Assets/FX/Materials/M_Generic_Ground_Fog_Sheet
graph {
 properties {
  bUsedWithNiagaraMeshParticles=true
  bUsedWithNiagaraSprites=true
  blendMode=BLEND_Translucent
  pixelDepthOffsetMode=PDOM_Legacy
  shadingModel=MSM_Unlit
  translucencyLightingMode=TLM_VolumetricPerVertexDirectional
  usedShadingModels=Unlit
 }
 nodes [
  n1 = TextureCoordinate()
  n2 = AppendVector(A=Add(A=ComponentMask(None=@n1) {r=true}, B=ParticleRandom()), B=ComponentMask(None=@n1) {g=true})
  n3 = Multiply(A=@n2, B=AppendVector(A=ScalarParameter() {defaultValue=1, parameterName=U1}, B=ScalarParameter() {defaultValue=0.33, parameterName=V1}))
  n4 = MaterialFunctionCall(f2=ActorPositionWS()) {materialFunction=&AddComponents}
  n5 = Multiply(A=Add(A=Divide(A=@n4.f2) {constB=100}, B=Time()), B=ScalarParameter() {defaultValue=1.2, parameterName=Panner Speed})
  n6 = Multiply(A=@n2, B=AppendVector(A=ScalarParameter() {defaultValue=1.4, parameterName=U2}, B=ScalarParameter() {defaultValue=0.25, parameterName=V2}))
  n7 = LinearInterpolate(
   A=TextureSample(UVs=Panner(Coordinate=@n3, Time=@n5) {bFractionalPart=true, speedX=0.0121, speedY=0.0534}) {samplerType=SAMPLERTYPE_LinearGrayscale, texture=/Game/Global/Textures/Default/Tileable_Patterns/T_StormCloudDetail01_M}.G
   B=TextureSample(UVs=Panner(Coordinate=@n6, Time=@n5) {bFractionalPart=true, speedX=-0.016, speedY=0.02}) {texture=/Game/Athena/SafeZone/S11/T_TilingClouds_D}.G
  )
  n8 = Multiply(A=@n7, B=@n7)
  Clamp(None=@n8) {minDefault=0.25}
  n9 = ParticleColor()
  MaterialFunctionCall(Emissive=@n9.RGB) {materialFunction=&MF_FX_EmissiveAdjustments}
  Multiply(A=@n8)
  n10 = Constant() {r=0.5}
  n11 = SphereMask(A=TextureCoordinate(), B=@n10, Radius=@n10) {hardnessPercent=0}
  n12 = Saturate(None=Add(A=@n8) {constB=0.15})
  n13 = Multiply(
   A=Multiply(A=Power(Base=@n11), B=@n12)
   B=MaterialFunctionCall("FadeSensitivity (optional)"=ScalarParameter() {defaultValue=0.4, parameterName=AngleFadeSensitivity}) {materialFunction=&FadeBasedOnViewAngle}.Result
  )
  n14 = LinearInterpolate(
   A=ScalarParameter() {defaultValue=50, parameterName=Depth Fade Distance2}
   B=ScalarParameter() {defaultValue=200, parameterName=Depth Fade Distance1}
   Alpha=@n12
  )
  n15 = Distance(A=ActorPositionWS(), B=CameraPositionWS())
  n16 = ScalarParameter() {defaultValue=-5000, parameterName=FarFadeOffset}
  n17 = Saturate(None=Divide(A=Add(A=@n15, B=@n16), B=ScalarParameter() {defaultValue=5000, parameterName=FarFadeDistance}))
  Multiply(A=@n9.A, B=Multiply(A=DepthFade(Opacity=@n13, FadeDistance=@n14), B=OneMinus(None=@n17)))
 ]
}
