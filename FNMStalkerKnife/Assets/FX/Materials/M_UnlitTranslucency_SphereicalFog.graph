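# Unlit translucent material for Niagara mesh particles that, as its name indicates, renders soft spherical fog.
# Opacity adds a world-position-mapped texture sample to a camera-to-normal dot product that softens silhouettes, then multiplies by particle alpha.
# A pixel-depth ramp fades it near the camera, and DepthFade softens intersections with scene geometry.
# Particle color feeds emissive through MF_FX_EmissiveAdjustments; a OneMinus node is left unconnected.
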Material /FNMStalkerKnife/Assets/FX/Materials/M_UnlitTranslucency_SphereicalFog
graph {
 properties {
  bAutomaticallySetUsageInEditor=false
  bUsedWithNiagaraMeshParticles=true
  blendMode=BLEND_Translucent
  shadingModel=MSM_Unlit
  translucencyLightingMode=TLM_VolumetricPerVertexDirectional
  usedShadingModels=Unlit
 }
 nodes [
  n1 = ParticleColor()
  n2 = Multiply(A=WorldPosition().XYZ, B=ScalarParameter() {defaultValue=0.0017, parameterName=Param_2})
  n3 = CameraVectorWS()
  n4 = VertexNormalWS()
  n5 = Add(
   A=Multiply(
    A=TextureSample(UVs=@n2) {samplerType=SAMPLERTYPE_LinearColor, texture=/Game/Global/MaterialFunctions/VT_ClothSimWPO}.RGB
    B=ScalarParameter() {defaultValue=0.222, parameterName=Param_3}
   )
   B=Saturate(None=DotProduct(A=@n3, B=@n4))
  )
  n6 = Divide(A=PixelDepth().R, B=ScalarParameter() {defaultValue=250, parameterName=Param_1})
  DepthFade(Opacity=Multiply(A=Multiply(A=@n1.A, B=@n5), B=Saturate(None=@n6)), FadeDistance=ScalarParameter() {defaultValue=25, parameterName=Param_6})
  MaterialFunctionCall(Emissive=@n1.RGB) {materialFunction=&MF_FX_EmissiveAdjustments}
  OneMinus()
 ]
}
